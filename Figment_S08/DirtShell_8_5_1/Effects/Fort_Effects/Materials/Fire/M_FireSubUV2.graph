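# Translucent unlit fire material for sprite and mesh particles, built on an 8x8 SubUV fluid flipbook.
# Color math uses particle color or a dynamic-parameter-driven blackbody with the flipbook.
# A panning tiled fire texture breaks up flipbook alpha, with dynamic parameters controlling noise blend, offset and floored scale, multiplied by particle alpha.
# Opacity gets noise-biased depth fade, bypassed on mobile, and an optional static-switch near-camera fade.
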Material /Figment_S08/DirtShell_8_5_1/Effects/Fort_Effects/Materials/Fire/M_FireSubUV2
graph {
 properties {
  bUsedWithMeshParticles=true
  bUsedWithParticleSprites=true
  blendMode=BLEND_Translucent
  pixelDepthOffsetMode=PDOM_Legacy
  shadingModel=MSM_Unlit
  usedShadingModels=Unlit
 }
 nodes [
  Constant() {r=3000.1265}
  LinearInterpolate()
  MaterialFunctionCall(Hue Shift Percentage=Constant() {r=0.6}, Texture=TextureSample() {texture=/Game/Effects/Fort_Effects/Textures/Grad/T_ColorGrad_FireRed}.RGB) {materialFunction=&HueShift}
  MaterialFunctionCall() {materialFunction=&3ColorBlend}
  n1 = DynamicParameter() {defaultValue=<rgba 0 0 0 1>, paramNames [LerpNoise, X, BlackBodyTemp, Scale]}
  n2 = TextureSampleParameterSubUV!(UVs=TextureCoordinate(), Apply View MipBias) {parameterName=FireSubUV, texture=/Game/Effects/Fort_Effects/Textures/Fire/T_FireFluid_1_8X8}
  Multiply(A=BlackBody!(Temp=@n1.BlackBodyTemp), B=@n2.RGB)
  n3 = ParticleColor()
  Multiply(A=@n3.RGB, B=Multiply(A=@n2.R, B=@n2.R))
  n4 = Add(A=TextureCoordinate() {vTiling=0.5}, B=AppendVector(A=@n1.X, B=Add(A=@n1.X) {constB=2}))
  n5 = Floor(None=@n1.Scale)
  n6 = TextureSample(UVs=Panner(Coordinate=Multiply(A=@n4, B=@n5) {desc=Scale}) {speedY=-0.125}) {texture=/Game/Effects/Fort_Effects/Textures/Fire/T_FireTile}
  n7 = Multiply(A=@n6.A, B=@n2.A)
  n8 = Multiply(A=@n3.A, B=LinearInterpolate(A=Add(A=@n7, B=@n2.A), B=Multiply(A=@n6.A, B=@n2.A), Alpha=@n1.LerpNoise))
  n9 = DepthFade(Opacity=@n8, FadeDistance=LinearInterpolate(Alpha=@n6.A) {constA=60, constB=120, desc=Bias Noise})
  n10 = FeatureLevelSwitch(Default=@n9, ES2=@n8, ES3_1=@n8, SM4_REMOVED=@n9) {desc="Added pref for mobile,Yang."}
  n11 = ScalarParameter() {defaultValue=150, parameterName=Near Camera Distance}
  StaticSwitchParameter(
   True=Multiply(A=@n10, B=MaterialFunctionCall(Distance=@n11) {materialFunction=&Near_Camera_Fade}.Result)
   False=@n10
  ) {parameterName=Use Near Camera Fade}
 ]
}
